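# Material function for the NeonTightSuit character material that draws animated, audio-reactive "dance lines" across base color, metallic, roughness, normal and emissive.
# Time-scrolled pre-skinned position drives gradient noise, which panning mask noise perturbs into repeating ramp coordinates.
# Rows of one ramp texture, sampled with explicit UV derivatives, supply color, metalness and stripe masks.
# MF_AudioReactivity provides amplitude and spectrum signals, while emissive adds hue-shifted rainbow stripes and normals flatten the incoming material's.
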MaterialFunction /Game/Global/Materials/MaterialGym/Materials/NeonTightSuit_v2/MF_DanceLines_v3
graph {nodes [
  n1 = SetMaterialAttributes!(
   MaterialAttributes
   ShadingModel
   BaseColor=NamedRerouteUsage()
   Metallic=NamedRerouteUsage()
   Specular
   Roughness=NamedRerouteUsage()
   Normal=NamedRerouteUsage()
   EmissiveColor=NamedRerouteUsage()
  ) {attributeSetTypes [
    #1
    #2
    #3
    #4
    #5
    #6
    #7
   ]}
  FunctionOutput(None=@n1)
  n2 = Multiply(A=Time(), B=Constant() {r=-1})
  n3 = Constant3Vector() {constant=<rgba 0 0 1 0>}
  n4 = LocalPosition() {localOrigin=InstancePreSkinning}
  n5 = ShaderStageSwitch(PixelShader=VertexInterpolator(VS=@n4.XYZ).PS, VertexShader=@n4.XYZ)
  n6 = NamedRerouteDeclaration(None=Add(A=Multiply(A=@n2, B=@n3), B=FunctionInput(Preview=@n5) {bUsePreviewValueAsDefault=true, inputName=Position})) {name=PreskinPos, nodeColor=<rgba 0.3828 1 0 1>}
  n7 = Multiply(
   A=SphereMask(
    A=MaterialFunctionCall(Float3=@n6) {materialFunction=&BreakOutFloat3Components}.G
    B=ScalarParameter() {defaultValue=150, desc=9.176, group=Dance Lines v3, parameterName=YAxisGradPosition}
    Radius=ScalarParameter() {defaultValue=50, desc=50, group=Dance Lines v3, parameterName=YAxisGradRadius}
    Hardness=Constant()
   )
   B=ScalarParameter() {defaultValue=1, group=Dance Lines v3, parameterName=YAxisGradIntensity}
  )
  n8 = SphereMask(
   A=@n6
   B=VectorParameter() {parameterName=ObjectSpaceGradStart, defaultValue=<rgba 0 40 150 1>}.RGB
   Radius=ScalarParameter() {defaultValue=180, group=Dance Lines, parameterName=ObjectSpaceGradRadius}
   Hardness=Constant()
  )
  n9 = ComponentMask(None=NamedRerouteUsage()) {b=true}
  n10 = ScalarParameter() {defaultValue=0.2, group=Dance Lines v3, parameterName=NormalDistortionWeight}
  Multiply(A=Add(A=Add(A=@n7, B=@n8), B=Multiply(A=@n9, B=@n10)), B=ScalarParameter() {group=Dance Lines v3, parameterName=StretchGradient})
  n11 = NamedRerouteUsage()
  n12 = Constant() {r=0.5}
  n13 = TextureCoordinate()
  n14 = DDX(Value=@n13)
  n15 = DDY(Value=@n13)
  NamedRerouteDeclaration(None=TextureSample!(
    UVs=MaterialFunctionCall(X=@n11, Y=@n12) {materialFunction=&MakeFloat2}.Result
    Tex=NamedRerouteUsage()
    "DDX(UVs)"=@n14
    "DDY(UVs)"=@n15
    Apply View MipBias
   ) {mipValueMode=TMVM_Derivative, samplerSource=SSM_Wrap_WorldGroupSettings}.G) {name=Metalness, nodeColor=<rgba 0 0.3437 1 1>}
  n16 = NamedRerouteUsage()
  n17 = TextureSampleParameter2D(UVs=Multiply(A=TextureCoordinate(), B=ScalarParameter() {defaultValue=4, group=Dance Lines v3, parameterName=GrainyNoiseTiling})) {
   group=Dance Lines v3
   parameterName=MaskTex
   samplerType=SAMPLERTYPE_Masks
   texture=&/T_NeonTightSuit_Mask
  }
  n18 = LinearInterpolate(B=@n17.G, Alpha=ScalarParameter() {defaultValue=1, group=Dance Lines v3, parameterName=GrainyNoiseIntensity}) {constA=1}
  n19 = NamedRerouteUsage()
  n20 = Multiply(A=LinearInterpolate(A=LinearInterpolate(A=@n18, Alpha=@n19), Alpha=NamedRerouteUsage()) {constB=0}, B=NamedRerouteUsage())
  NamedRerouteDeclaration(None=Multiply(A=Multiply(A=@n16, B=@n20), B=ScalarParameter() {defaultValue=2, group=Dance Lines v3, parameterName=Brightness})) {name=Diffuse, nodeColor=<rgba 0 1 0.6953 1>}
  n21 = ScalarParameter() {defaultValue=1, group=Dance Lines v3, parameterName=Repetition}
  n22 = NamedRerouteUsage()
  n23 = Multiply(
   A=ComponentMask(None=Noise!(World Position=@n22, FilterWidth) {bTurbulence=false, levels=1, noiseFunction=NOISEFUNCTION_GradientALU, outputMin=0, scale=0.04}) {g=true}
   B=ScalarParameter() {defaultValue=1, group=Dance Lines v3, parameterName=NoiseWarpIntensity}
  )
  n24 = OneMinus(None=@n23)
  n25 = ScalarParameter() {defaultValue=0.15, desc=0.25, group=Dance Lines v3, parameterName=ReactiveAmp}
  n26 = Saturate(None=Subtract(A=NamedRerouteUsage(), B=NamedRerouteUsage()))
  n27 = Multiply(A=ComponentMask(None=NamedRerouteUsage()) {b=true, r=true})
  n28 = Constant2Vector() {desc=-0.25, g=0.3333, r=-0.25}
  n29 = ScalarParameter() {defaultValue=3, group=Dance Lines v3, parameterName=PanningSpeed}
  n30 = TextureObject() {samplerType=SAMPLERTYPE_Masks, texture=&T_NeonTightSuit_Mask}
  n31 = TextureSample!(
   UVs=Panner(Coordinate=Divide(A=@n27) {constB=32, desc=40}, Speed=Multiply(A=@n28, B=@n29)) {bFractionalPart=true, speedX=-0.45, speedY=0.35}
   Tex=@n30
   Apply View MipBias
  ) {samplerSource=SSM_Wrap_WorldGroupSettings}
  n32 = Multiply(A=@n24, B=Multiply(A=@n31.R, B=LinearInterpolate(B=@n25, Alpha=@n26) {constB=0.5}) {constB=0.2}) {desc=Ease in at start}
  n33 = Multiply(A=@n23, B=Add(A=@n21, B=NamedRerouteDeclaration(None=@n32) {name=Noise, nodeColor=<rgba 0.0312 0 1 1>}) {constB=0})
  n34 = Multiply(A=Time(), B=ScalarParameter() {defaultValue=0.05, group=Dance Lines v3, parameterName=Speed}) {constB=0}
  NamedRerouteDeclaration(None=Frac(None=Add(A=@n33, B=@n34))) {name=RampX_Colored, nodeColor=<rgba 0.3125 1 0 1>}
  NamedRerouteDeclaration() {name=Debug, nodeColor=<rgba 0.1953 0 1 1>}
  n35 = Multiply(A=NamedRerouteUsage(), B=ScalarParameter() {defaultValue=0.05, group=Dance Lines v3, parameterName=AmplitudeHueShiftStrength})
  n36 = Multiply(A=Time()) {constB=0.1, desc=hue shift speed. was 0.25}
  NamedRerouteDeclaration(None=Frac(None=Add(A=@n35, B=@n36))) {name=HueShiftPercentage, nodeColor=<rgba 0.5937 1 0 1>}
  n37 = TextureSample!(
   UVs=MaterialFunctionCall(X=@n11, Y=Constant() {r=0.1678}) {materialFunction=&MakeFloat2}.Result
   Tex=NamedRerouteUsage()
   "DDX(UVs)"=@n14
   "DDY(UVs)"=@n15
   Apply View MipBias
  ) {mipValueMode=TMVM_Derivative, samplerSource=SSM_Wrap_WorldGroupSettings}
  NamedRerouteDeclaration(None=@n37.RGB) {name=RampColor, nodeColor=<rgba 1 0 0.9141 1>}
  n38 = NamedRerouteUsage()
  n39 = Constant() {r=0.8333}
  NamedRerouteDeclaration(None=TextureSample(
    UVs=MaterialFunctionCall(X=@n38, Y=@n39) {materialFunction=&MakeFloat2}.Result
    Tex=NamedRerouteUsage()
    "DDX(UVs)"=@n14
    "DDY(UVs)"=@n15
   ) {mipValueMode=TMVM_Derivative, samplerSource=SSM_Wrap_WorldGroupSettings}.R) {name=RampMask_Colored, nodeColor=<rgba 0 1 0.7656 1>}
  n40 = TextureObjectParameter() {group=Dance Lines v3, parameterName=RampTex, texture=&/T_NeonTightSuit_Dark_v2}
  NamedRerouteDeclaration(None=@n40) {name=RampTex, nodeColor=<rgba 1 0.4453 0 1>}
  n41 = TextureSample(
   UVs=MaterialFunctionCall(X=NamedRerouteUsage(), Y=Constant() {r=0.8333}) {materialFunction=&MakeFloat2}.Result
   Tex=NamedRerouteUsage()
   "DDX(UVs)"=@n14
   "DDY(UVs)"=@n15
  ) {mipValueMode=TMVM_Derivative, samplerSource=SSM_Wrap_WorldGroupSettings}
  NamedRerouteDeclaration(None=@n41.G) {name=RampMask_ClearCoat, nodeColor=<rgba 0.5 1 0 1>}
  n42 = Constant2Vector() {g=-0.2555, r=0.1}
  n43 = TextureSample!(
   UVs=Panner(Coordinate=Divide(A=@n27) {constB=50}, Speed=Multiply(A=@n29, B=@n42)) {bFractionalPart=true, speedX=2, speedY=-0.5}
   Tex=@n30
   Apply View MipBias
  ) {samplerSource=SSM_Wrap_WorldGroupSettings}
  NamedRerouteDeclaration(None=Power(Base=@n43.B) {constExponent=3}) {name=Noise2, nodeColor=<rgba 0.8047 1 0 1>}
  n44 = MaterialFunctionCall() {materialFunction=/Game/Packages/Fortress_SharedMaterials/Base_Material_Functions/BaseCharacterShaderFunctions/Audio/MF_AudioReactivity}
  n45 = ScalarParameter() {defaultValue=8, group=Dance Lines v3, parameterName=Audio_Amplitude_Exp}
  NamedRerouteDeclaration(None=SmoothStep(Value=Power(Base=@n44.Amplitude, Exp=@n45) {constExponent=3})) {name=Amplitude, nodeColor=<rgba 0 1 0.4609 1>}
  n46 = MaterialFunctionCall(Float4=@n44.Spectrum) {materialFunction=&BreakOutFloat4Components}
  n47 = ScalarParameter() {defaultValue=1.5, group=Dance Lines v3, parameterName=Audio_HighFreq_Exp}
  NamedRerouteDeclaration(None=SmoothStep(Value=Power(Base=@n46.B, Exp=@n47) {constExponent=3})) {name=SpectrumB, nodeColor=<rgba 1 0.1641 0 1>}
  n48 = NamedRerouteUsage()
  n49 = NamedRerouteUsage()
  n50 = LinearInterpolate(
   B=ScalarParameter() {defaultValue=0.8, desc=0.25, group=Dance Lines v3, parameterName=ReactiveAmp2}
   Alpha=Saturate(None=Subtract(A=@n48, B=@n49))
  ) {constB=0.3}
  n51 = NamedRerouteDeclaration(None=Multiply(A=@n24, B=Multiply(A=@n43.R, B=@n50) {constB=0.2}) {desc=Ease in at start}) {name=Noise_1, nodeColor=<rgba 0.0312 0 1 1>}
  n52 = Add(A=Multiply(A=@n23, B=Add(A=@n21, B=@n51) {constB=0}), B=@n34)
  n53 = Multiply(A=ScalarParameter() {defaultValue=-0.02, group=Dance Lines v3, parameterName=Offset}, B=NamedRerouteUsage())
  NamedRerouteDeclaration(None=Frac(None=Add(A=@n52, B=@n53) {constB=0.2})) {name=RampX_Colored_1, nodeColor=<rgba 0.3125 1 0 1>}
  NamedRerouteDeclaration(None=@n41.R) {name=RampMask_Colored_1, nodeColor=<rgba 0 1 0.7656 1>}
  NamedRerouteDeclaration(None=Power(Base=@n46.G)) {name=SpectrumG, nodeColor=<rgba 0.125 1 0 1>}
  n54 = LinearInterpolate(
   A=Multiply(A=NamedRerouteUsage(), B=ScalarParameter() {defaultValue=0.25, group=Dance Lines v3, parameterName=EmissiveIntensity})
   Alpha=NamedRerouteUsage()
  ) {constB=0}
  n55 = NamedRerouteUsage()
  n56 = ComponentMask(None=NamedRerouteUsage()) {b=true, r=true}
  n57 = TextureSampleParameter2D(UVs=Panner(Coordinate=Divide(A=@n56) {constB=128}, Speed=Constant2Vector() {g=0.012, r=0.01}) {bFractionalPart=true}) {
   group=Dance Lines v3
   parameterName=StripesColor
   texture=&/T_NeonTightSuit_Rainbow_Dark
  }
  n58 = Multiply(A=NamedRerouteUsage(), B=@n57.RGB)
  n59 = Frac(None=Add(A=@n55) {constB=0.15})
  n60 = NamedRerouteUsage()
  n61 = Multiply(A=@n57.RGB, B=@n60)
  n62 = Multiply(
   A=LinearInterpolate(
    A=MaterialFunctionCall(Hue Shift Percentage=@n55, Texture=@n58) {materialFunction=&HueShift}.Result
    B=MaterialFunctionCall(Hue Shift Percentage=@n59, Texture=@n61) {materialFunction=&HueShift}.Result
    Alpha=@n60
   )
   B=ScalarParameter() {defaultValue=2, group=Dance Lines v3, parameterName=ColorStripesEmissive}
  )
  NamedRerouteDeclaration(None=Multiply(A=NamedRerouteUsage(), B=Add(A=@n54, B=@n62))) {name=Emissive, nodeColor=<rgba 1 0.1406 0 1>}
  n63 = DotProduct(A=PixelNormalWS(), B=CameraVectorWS())
  NamedRerouteDeclaration(None=Power(Base=Clamp(None=@n63), Exp=ScalarParameter() {defaultValue=2, group=Dance Lines v3, parameterName=ViewDepedantFalloff})) {name=Fresnel, nodeColor=<rgba 1 0.5859 0 1>}
  n64 = ScalarParameter() {defaultValue=0.1, group=Dance Lines v3, parameterName=RoughnessStripesThreshold}
  n65 = Clamp(None=Multiply(A=SphereMask(A=@n37.R, B=@n64) {attenuationRadius=0.1, hardnessPercent=1}) {constB=2.5}) {maxDefault=0.8}
  n66 = ScalarParameter() {defaultValue=0.2, group=Dance Lines v3, parameterName=MetalStripes_Roughness}
  n67 = NamedRerouteUsage()
  n68 = Saturate(None=Add(A=NamedRerouteUsage(), B=NamedRerouteUsage()) {constB=0})
  NamedRerouteDeclaration(None=LinearInterpolate(
    A=LinearInterpolate(A=@n65, B=@n66, Alpha=@n67) {constA=0.5, constB=0}
    Alpha=NamedRerouteDeclaration(None=@n68) {name=RampMask_Colored_combined, nodeColor=<rgba 1 0.2344 0 1>}
   ) {constB=0.1}) {name=Roughness, nodeColor=<rgba 0.2656 0 1 1>}
  NamedRerouteDeclaration(None=MaterialFunctionCall(Normal=NamedRerouteUsage(), Flatness=ScalarParameter() {defaultValue=0.5, group=Dance Lines v3, parameterName=DanceLines_NormalFlattness}) {materialFunction=&FlattenNormal}.Result) {name=Normal, nodeColor=<rgba 0.0547 1 0 1>}
  NamedRerouteDeclaration(None=GetMaterialAttributes(None=FunctionInput() {bUsePreviewValueAsDefault=true, inputName=Material, inputType=FunctionInput_MaterialAttributes}) {attributeGetTypes [#6]}.Normal) {name=Normal_IN, nodeColor=<rgba 1 0 0.375 1>}
  n69 = Multiply(A=@n31.R, B=NamedRerouteUsage()) {constB=0.01}
  n70 = ScalarParameter() {defaultValue=0.04, group=Dance Lines v3, parameterName=AmplitudeRepetitionStrength}
  n71 = Multiply(A=Add(A=Multiply(A=@n69, B=@n70) {constB=0}, B=@n21) {constB=0}, B=@n23)
  NamedRerouteDeclaration(None=Frac(None=Add(A=@n71, B=@n34))) {name=RampX, nodeColor=<rgba 0.3125 1 0 1>}
  NamedRerouteUsage()
  StaticSwitchParameter(True=NamedRerouteUsage()) {parameterName=DebugOn}
 ]}
